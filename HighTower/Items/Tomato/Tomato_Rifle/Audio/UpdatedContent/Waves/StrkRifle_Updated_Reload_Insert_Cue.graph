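# Sound cue for the magazine-insert step of the Tomato rifle's reload.
# A single wave player feeds one mag-in wave straight to the output, with no randomization or mixing, at lowered pitch and volume.
# It uses the shared weapon reload attenuation and weapon foley sound class, overrides concurrency, and has the debug flag left on.
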SoundCue /HighTower/Items/Tomato/Tomato_Rifle/Audio/UpdatedContent/Waves/StrkRifle_Updated_Reload_Insert_Cue
root=WavePlayer_0
nodes [{
  id=WavePlayer_0
  properties {soundWaveAssetPtr=&/StrkRifle_Updated_Reload_MagIn_01}
 }]
properties {
 attenuationSettings=&Fort_Weapon_Reload_Att
 bDebug=true
 bOverrideConcurrency=true
 maxDistance=2250
 pitchMultiplier=0.9
 soundClassObject=&Fort_WeaponFoley
 volumeMultiplier=0.7594
}
