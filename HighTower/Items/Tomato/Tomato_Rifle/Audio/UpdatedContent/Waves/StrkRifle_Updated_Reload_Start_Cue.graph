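# Sound cue for the start of the Tomato rifle's updated reload sound.
# A mixer at the root layers the rifle's magazine-out wave with a reload start wave borrowed from the burst rifle, playing both together at equal volume.
# It uses weapon reload attenuation and the weapon foley sound class, lowers overall volume, overrides concurrency and has its debug flag enabled.
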SoundCue /HighTower/Items/Tomato/Tomato_Rifle/Audio/UpdatedContent/Waves/StrkRifle_Updated_Reload_Start_Cue
root=Mixer_0
nodes [
 {id=Mixer_0, children [WavePlayer_0,WavePlayer_1], properties {inputVolume [1,1]}}
 {
  id=WavePlayer_0
  properties {soundWaveAssetPtr=&/StrkRifle_Updated_Reload_Magout_01}
 }
 {
  id=WavePlayer_1
  properties {soundWaveAssetPtr=/Game/Sounds/Fort_Weapons/AutoRifle/BurstRifle/Burstrifle_HT_Reload_Start.BurstRifle_HT_Reload_Start}
 }
]
properties {
 attenuationSettings=&Fort_Weapon_Reload_Att
 bDebug=true
 bOverrideConcurrency=true
 maxDistance=2250
 soundClassObject=&Fort_WeaponFoley
 volumeMultiplier=0.6254
}
